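# Animation Blueprint for the Drift Winter backpack that animates its four floating icicles.
# The event graph drives sine-wave bone-space offsets at separate top and bottom frequencies, applied additively to the icicle joints and cached.
# Two AnimDynamics chains on the master joints read that cache.
# The output blends to the spring-free, damped chain while moving on the ground, and to the linear-spring, zero-gravity chain otherwise.
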AnimBlueprint /BRCosmetics/Accessories/FORT_Backpacks/Backpack_M_MED_Drift_Winter/Meshes/M_MED_Drift_Winter_Pack_AnimBP
target_skeleton=&/M_MED_Drift_Winter_Pack_Skeleton
graphs [
 AnimGraph {nodes [
   n1 = Anim Dynamics - Bone: dyn_L_icicle_top_master_jnt(ComponentPose=Local To Component(LocalPose=Use cached pose 'transformCache'()), Alpha=1, ExternalForce="0,0,0") {
    angularDampingOverride=0.9
    angularSpringConstant=200
    bGravityOverrideInSimSpace=true
    bLinearSpring=true
    bOverrideAngularDamping=true
    bOverrideLinearDamping=true
    gravityScale=0
    linearDampingOverride=0.9999
    linearSpringConstant=100
    simulationSpace=RootRelative
    alphaBoolBlend {blendOption=Linear}
    boundBone {boneName=dyn_L_icicle_top_master_jnt}
    componentAppliedLinearAccClamp=<50000 50000 50000>
    componentLinearAccScale {x=0.1, z=0.1}
    componentLinearVelScale=<0.01 0.01 0.01>
    physicsBodyDefinitions [{
      sphereCollisionRadius=0
      boundBone {boneName=dyn_L_icicle_top_master_jnt}
      boxExtents=<4 4 4>
      constraintSetup {angularLimitsMax=<0 0 0>, angularLimitsMin=<0 0 0>, angularTarget=<0 0 0>, linearAxesMax=<4.3 4.3 4.3>, linearAxesMin=<-0.5 -0.1 -0.5>}
      localJointOffset=<0 0 0>
     }]
    relativeSpaceBone {boneName=root}
   }
   n2 = Anim Dynamics - Bone: dyn_R_icicle_top_master_jnt(ComponentPose=@n1, Alpha=1, ExternalForce="0,0,0") {
    angularDampingOverride=0.9
    angularSpringConstant=200
    bGravityOverrideInSimSpace=true
    bLinearSpring=true
    bOverrideAngularDamping=true
    bOverrideLinearDamping=true
    gravityScale=0
    linearDampingOverride=0.9999
    linearSpringConstant=100
    simulationSpace=RootRelative
    alphaBoolBlend {blendOption=Linear}
    boundBone {boneName=dyn_R_icicle_top_master_jnt}
    componentAppliedLinearAccClamp=<30000 30000 30000>
    componentLinearAccScale {x=0.1, z=0.1}
    componentLinearVelScale=<0.01 0.2 0.01>
    physicsBodyDefinitions [{
      sphereCollisionRadius=0
      boundBone {boneName=dyn_R_icicle_top_master_jnt}
      boxExtents=<4 4 4>
      constraintSetup {angularLimitsMax=<0 0 0>, angularLimitsMin=<0 0 0>, angularTarget=<0 0 0>, linearAxesMax=<5 5 5>, linearAxesMin=<-0.5 -0.1 -0.5>}
      localJointOffset=<0 0 0>
     }]
    relativeSpaceBone {boneName=root}
   }
   n3 = Anim Dynamics - Bone: dyn_L_icicle_bottom_master_jnt(ComponentPose=@n2, Alpha=1, ExternalForce="0,0,0") {
    angularDampingOverride=0.9
    angularSpringConstant=200
    bGravityOverrideInSimSpace=true
    bLinearSpring=true
    bOverrideAngularDamping=true
    bOverrideLinearDamping=true
    gravityScale=0
    linearDampingOverride=0.9999
    linearSpringConstant=100
    simulationSpace=RootRelative
    alphaBoolBlend {blendOption=Linear}
    boundBone {boneName=dyn_L_icicle_bottom_master_jnt}
    componentAppliedLinearAccClamp=<50000 50000 50000>
    componentLinearAccScale=<0.03 0.01 0.03>
    componentLinearVelScale=<0.03 0.03 0.03>
    physicsBodyDefinitions [{
      sphereCollisionRadius=0
      boundBone {boneName=dyn_L_icicle_bottom_master_jnt}
      boxExtents=<4 4 4>
      constraintSetup {angularLimitsMax=<0 0 0>, angularLimitsMin=<0 0 0>, angularTarget=<0 0 0>, linearAxesMax=<3.5 3.5 3.5>, linearAxesMin=<-0.5 -0.5 -0.5>}
      localJointOffset=<0 0 0>
     }]
    relativeSpaceBone {boneName=root}
   }
   n4 = Anim Dynamics - Bone: dyn_R_icicle_bottom_master_jnt(ComponentPose=@n3, Alpha=1, ExternalForce="0,0,0") {
    angularDampingOverride=0.9
    angularSpringConstant=200
    bGravityOverrideInSimSpace=true
    bLinearSpring=true
    bOverrideAngularDamping=true
    bOverrideLinearDamping=true
    gravityScale=0
    linearDampingOverride=0.9999
    linearSpringConstant=100
    simulationSpace=RootRelative
    alphaBoolBlend {blendOption=Linear}
    boundBone {boneName=dyn_R_icicle_bottom_master_jnt}
    componentAppliedLinearAccClamp=<50000 50000 50000>
    componentLinearAccScale=<0.03 0.01 0.03>
    componentLinearVelScale=<0.03 0.03 0.03>
    physicsBodyDefinitions [{
      sphereCollisionRadius=0
      boundBone {boneName=dyn_R_icicle_bottom_master_jnt}
      boxExtents=<4 4 4>
      constraintSetup {angularLimitsMax=<0 0 0>, angularLimitsMin=<0 0 0>, angularTarget=<0 0 0>, linearAxesMax=<3.5 3.5 3.5>, linearAxesMin=<-0.5 -0.5 -0.5>}
      localJointOffset=<0 0 0>
     }]
    relativeSpaceBone {boneName=root}
   }
   n5 = Anim Dynamics - Bone: dyn_L_icicle_top_master_jnt(ComponentPose=Local To Component(LocalPose=Use cached pose 'transformCache'()), Alpha=1, ExternalForce="0,0,0") {
    angularDampingOverride=0.98
    bGravityOverrideInSimSpace=true
    bOverrideAngularDamping=true
    bOverrideLinearDamping=true
    linearDampingOverride=0.98
    simulationSpace=RootRelative
    alphaBoolBlend {blendOption=Linear}
    boundBone {boneName=dyn_L_icicle_top_master_jnt}
    physicsBodyDefinitions [{
      sphereCollisionRadius=0
      boundBone {boneName=dyn_L_icicle_top_master_jnt}
      boxExtents=<1 1 1>
      constraintSetup {angularLimitsMax=<0 0 0>, angularLimitsMin=<0 0 0>, angularTarget=<0 0 0>}
      localJointOffset=<0 0 0>
     }]
    relativeSpaceBone {boneName=root}
   }
   n6 = Anim Dynamics - Bone: dyn_R_icicle_top_master_jnt(ComponentPose=@n5, Alpha=1, ExternalForce="0,0,0") {
    angularDampingOverride=0.98
    bGravityOverrideInSimSpace=true
    bOverrideAngularDamping=true
    bOverrideLinearDamping=true
    linearDampingOverride=0.98
    simulationSpace=RootRelative
    alphaBoolBlend {blendOption=Linear}
    boundBone {boneName=dyn_R_icicle_top_master_jnt}
    physicsBodyDefinitions [{
      sphereCollisionRadius=0
      boundBone {boneName=dyn_R_icicle_top_master_jnt}
      boxExtents=<1 1 1>
      constraintSetup {angularLimitsMax=<0 0 0>, angularLimitsMin=<0 0 0>, angularTarget=<0 0 0>}
      localJointOffset=<0 0 0>
     }]
    relativeSpaceBone {boneName=root}
   }
   n7 = Anim Dynamics - Bone: dyn_L_icicle_bottom_master_jnt(ComponentPose=@n6, Alpha=1, ExternalForce="0,0,0") {
    angularDampingOverride=0.98
    bGravityOverrideInSimSpace=true
    bOverrideAngularDamping=true
    bOverrideLinearDamping=true
    linearDampingOverride=0.98
    simulationSpace=RootRelative
    alphaBoolBlend {blendOption=Linear}
    boundBone {boneName=dyn_L_icicle_bottom_master_jnt}
    physicsBodyDefinitions [{
      sphereCollisionRadius=0
      boundBone {boneName=dyn_L_icicle_bottom_master_jnt}
      boxExtents=<1 1 1>
      constraintSetup {angularLimitsMax=<0 0 0>, angularLimitsMin=<0 0 0>, angularTarget=<0 0 0>}
      localJointOffset=<0 0 0>
     }]
    relativeSpaceBone {boneName=root}
   }
   n8 = Anim Dynamics - Bone: dyn_R_icicle_bottom_master_jnt(ComponentPose=@n7, Alpha=1, ExternalForce="0,0,0") {
    angularDampingOverride=0.98
    bGravityOverrideInSimSpace=true
    bOverrideAngularDamping=true
    bOverrideLinearDamping=true
    linearDampingOverride=0.98
    simulationSpace=RootRelative
    alphaBoolBlend {blendOption=Linear}
    boundBone {boneName=dyn_R_icicle_bottom_master_jnt}
    physicsBodyDefinitions [{
      sphereCollisionRadius=0
      boundBone {boneName=dyn_R_icicle_bottom_master_jnt}
      boxExtents=<1 1 1>
      constraintSetup {angularLimitsMax=<0 0 0>, angularLimitsMin=<0 0 0>, angularTarget=<0 0 0>}
      localJointOffset=<0 0 0>
     }]
    relativeSpaceBone {boneName=root}
   }
   n9 = Get bIsMoving2D()
   n10 = Get bIsOnGround() -> (bIsOnGround:bool)
   Output Pose(Result=Blend Poses by bool(
     BlendPose_0=Component To Local(ComponentPose=@n8)
     BlendPose_1=Component To Local(ComponentPose=@n4)
     BlendTime_0=0.75
     BlendTime_1=0.75
     bActiveValue=AND Boolean(A=@n10, B=@n9)
    ) {blendType=Linear}) {layerGroup=None}
   n11 = "Get Translate Floaters (Top)"() -> ("Translate Floaters (Top)":Vector)
   n12 = "Transform (Modify) Bone - Bone: dyn_L_icicle_top_jnt"(ComponentPose=Local To Component(LocalPose=Input Pose()), Alpha=1, Translation=@n11, Rotation="0,0,0", Scale="1,1,1") {
    rotationMode=BMM_Additive
    scaleSpace=BCS_BoneSpace
    translationMode=BMM_Additive
    translationSpace=BCS_BoneSpace
    alphaBoolBlend {blendOption=Linear}
    boneToModify {boneName=dyn_L_icicle_top_jnt}
   }
   n13 = "Get Translate Floaters (Bottom)"() -> ("Translate Floaters (Bottom)":Vector)
   n14 = "Transform (Modify) Bone - Bone: dyn_L_icicle_bottom_jnt"(ComponentPose=@n12, Alpha=1, Translation=@n13, Rotation="0,0,0", Scale="1,1,1") {
    rotationMode=BMM_Additive
    scaleSpace=BCS_BoneSpace
    translationMode=BMM_Additive
    translationSpace=BCS_BoneSpace
    alphaBoolBlend {blendOption=Linear}
    boneToModify {boneName=dyn_L_icicle_bottom_jnt}
   }
   n15 = "Transform (Modify) Bone - Bone: dyn_R_icicle_top_jnt"(ComponentPose=@n14, Alpha=1, Translation=@n11, Rotation="0,0,0", Scale="1,1,1") {
    rotationMode=BMM_Additive
    scaleSpace=BCS_BoneSpace
    translationMode=BMM_Additive
    translationSpace=BCS_BoneSpace
    alphaBoolBlend {blendOption=Linear}
    boneToModify {boneName=dyn_R_icicle_top_jnt}
   }
   n16 = "Transform (Modify) Bone - Bone: dyn_R_icicle_bottom_jnt"(ComponentPose=@n15, Alpha=1, Translation=@n13, Rotation="0,0,0", Scale="1,1,1") {
    rotationMode=BMM_Additive
    scaleSpace=BCS_BoneSpace
    translationMode=BMM_Additive
    translationSpace=BCS_BoneSpace
    alphaBoolBlend {blendOption=Linear}
    boneToModify {boneName=dyn_R_icicle_bottom_jnt}
   }
   Save cached pose 'transformCache'(Pose=Component To Local(ComponentPose=@n16)) {cacheName=transformCache}
  ]}
 EventGraph {nodes [
   Event BlueprintUpdateAnimation()
   > n1 = Sequence()
   n2 = "Sin (Radians)"(A=float * float(A=GetRealTimeSeconds(), B=1.6))
   > "Set Translate Floaters (Top)"("Translate Floaters (Top)":Vector=MakeVector(X=float / float(A=@n2, B=1.35) {is=title K2Node_CallFunction}, Y=0)) -> (Output_Get:Vector)
   n3 = "Sin (Radians)"(A=float * float(A=GetRealTimeSeconds(), B=1.1))
   "Set Translate Floaters (Bottom)"(execute=@n1.then_1, "Translate Floaters (Bottom)":Vector=MakeVector(X=float / float(A=@n3, B=1.35), Y=0)) -> (Output_Get:Vector)
   TryGetPawnOwner()
  ]}
]
variables [
 "Translate Floaters (Bottom)" {type=Vector}
 "Translate Floaters (Top)" {type=Vector}
]
